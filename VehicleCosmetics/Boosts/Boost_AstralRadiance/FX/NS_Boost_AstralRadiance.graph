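# Niagara system for the Astral Radiance vehicle boost cosmetic, emitting trails and star bits from the car's skeletal boost sockets.
# A ribbon emitter spawns quality-scaled trails per port, with two renderers and separate curve-shaped width and color for a main and a thin trail.
# A sprite emitter spreads shrinking, curl-noise-driven bits across the sockets with randomized material parameters.
# The painted-variant flag switches each layer between built-in colors and user-supplied colors.
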NiagaraSystem /VehicleCosmetics/Boosts/Boost_AstralRadiance/FX/NS_Boost_AstralRadiance
dependencies {
 usedDataInterfaces [
  &NiagaraDataInterfaceCurve
  &NiagaraDataInterfaceVector2DCurve
  &NiagaraDataInterfaceVectorCurve
  &NiagaraDataInterfaceColorCurve
  &NiagaraDataInterfaceSkeletalMesh
 ]
 usedDynamicInputs []
 usedModules [
  &self:SystemSpawnScript.SystemScriptSource.SystemScriptGraph.NiagaraNodeAssignment_0.SetVariables_#1
  &SystemState
 ]
}
emitters {
 NE_VC_Boost_Ribbon_Parent001 {
  emitterUpdateScript [
   &EmitterState()
   &VC_FilteredSocketCount()
   &VC_RateQualityHandler(Epic Spawn Rate=120, High Spawn Rate=90, Low Spawn Rate=50, Medium Spawn Rate=60, Use Custom Rate?=true)
   &VC_SpawnRibbon_v1(Number of Ribbons=$Emitter.PortCount, Spawn Time=-1, SpawnRate=$Emitter.Spawn Rate)
   /SetVariables(Emitter.BTrailWidth=25) {isSetParameters=true}
  ]
  particleSpawnScript [
   &VC_PaintedVariantsOverride() {enabled=false}
   &VC_InitializeRibbon_v1(
    Color=<rgba 1 1 5 1>
    Custom Offset=true
    Enable Skeletal Position=true
    Lifetime=0.325
    Local Vector Offset=&AddVectorToPosition(Position=<30 0 0>, Vector=$User.DTCarOffset)
    Offset Amount=1
    Ribbon Width=35
   )
   &AddVelocity(Velocity=<-4000 0 0>) {module=AddVelocity001}
  ]
  particleUpdateScript [
   &ParticleState()
   &ScaleColor(
    Scale RGB=&VectorFromCurve(
     VectorCurve=DataInterface() {propertyValues {
       XCurve {keys [<at 0.3 4.25>,<at 0.4 0>,<at 0.5 0.125>,<at 1 0.3>]}
       YCurve {keys [<at 0.02 1>,<at 0.3 2.25>,<at 0.4 4.25>,<at 0.61 6>,<at 1 0.5>]}
       ZCurve {keys [<at 0.02 1>,<at 0.3 3.25>,<at 0.4 1.75>,<at 0.5 1.75>,<at 0.61 1>,<at 1 0.8>]}
      }}
     Scale Curve=<0.1 0.1 0.1>
    )
    ScaleA=true
    ScaleRGB=true
   )
   &Color(Color=&MakeCustomLinearColorFromBool(
     Bool=$User.bIsPainted
     True Color=$User.PaintedColor
     False Color=&ColorFromCurve(ColorCurve=DataInterface() {propertyValues {
        AlphaCurve {keys [<at 0 1 RCIM_Cubic>,<at 0.2942 0.9873>,<at 0.8359 0.896>,<at 1 0 RCIM_Cubic>]}
        BlueCurve {keys [<at 0 5 RCIM_Cubic>,<at 0.0886 1.5316>,<at 0.1883 1>,<at 0.3044 1>,<at 0.3879 0.5>,<at 0.5101 0.5>,<at 0.7178 0.8429>,<at 0.9031 1.024>]}
        GreenCurve {keys [<at 0 1 RCIM_Cubic>,<at 0.0886 0.2055>,<at 0.1883 0.1882>,<at 0.3044 0.1882>,<at 0.3879 0.4167>,<at 0.5101 0.375>,<at 0.7178 0.1944>,<at 0.9031 -0.0395>]}
        RedCurve {keys [<at 0 3.3333 RCIM_Cubic>,<at 0.0886 0.6339>,<at 0.1883 0>,<at 0.3044 0>,<at 0.3879 0>,<at 0.5101 0>,<at 0.7178 0.1927>,<at 0.9031 0.1324>]}
       }})
    ))
   &ScaleRibbonWidth(Ribbon Width Scale=&VectorLength(Vector=&VectorFromCurve(VectorCurve=DataInterface() {propertyValues {
        XCurve {keys [<at 0 1.63>,<at 0.5 1.05>,<at 1 0.25>]}
        YCurve {keys [<at 0 1.5>,<at 0.5 0.85>,<at 1 0.25>]}
        ZCurve {keys [<at 0 1.5>,<at 0.5 0.85>,<at 1 0.25>]}
       }})))
   &SolveForcesAndVelocity(Clamp Velocity=false, Limit Acceleration=false)
   &VC_UpdateRibbon_v1(Enable Skeletal Position=true)
   /SetVariables(
    Particles.BTrailScaleColor:LinearColor=&VC_DynamicScaleColor(
     Color to Scale:LinearColor=&MakeCustomLinearColorFromBool(Bool=$User.bIsPainted, True Color=$User.ThinTrailColor, False Color=<rgba 0.4717 4.7013 5 1>)
     RGB Scale:Vector3f=<1 1 1>
     Alpha Scale:NiagaraFloat=1
    )
    Particles.BTrailScaleWidth:NiagaraFloat=&Multiply_Float(
     A=$Emitter.BTrailWidth
     B=&FloatFromCurve(FloatCurve=DataInterface() {propertyValues {Curve {keys [<at 0 2>,<at 1 0>]}}})
    )
   ) {isSetParameters=true}
  ]
  rendererClasses [&NiagaraRibbonRendererProperties]
  renderers [
   {rendererClass=&NiagaraRibbonRendererProperties}
   {rendererClass=&NiagaraRibbonRendererProperties, rendererIndex=1}
  ]
 }
 NE_VC_Boost_Sprite_Parent {
  emitterUpdateScript [
   &EmitterState()
   &VC_FilteredSocketCount()
   &SpawnRate(SpawnRate=20, Use Spawn Probability=false)
  ]
  particleSpawnScript [
   &VC_PaintedVariantsOverride() {enabled=false}
   &InitializeParticle(
    Color Mode=::NewEnumerator0
    Lifetime=&RandomRangeFloat(Minimum=1.25, Maximum=1.75)
    Sprite Rotation Mode=::NewEnumerator2
    Sprite Size Mode=::NewEnumerator3
    Uniform Sprite Size=&RandomRangeFloat(Minimum=5, Maximum=16)
    UsePositionOffset=false
    Write Material Random=true
   )
   &PartitionParticles(Number Of Partitions=$Emitter.PortCount)
   &SkeletalMeshLocation(
    Bone / Socket Index=$Particles.PartitionParticles.Partition
    Bone / Socket Sampling Mode=::NewEnumerator5
    MaskSpawnGroup=false
    Mesh Sampling Type=::NewEnumerator6
    Sampled Position Offset=<-50 0 0>
    Skeletal Mesh=$System.SkeletalMesh
   )
   &AddVelocity(Velocity=<-400 0 0>)
   &ShapeLocation(
    MaskSpawnGroup=false
    Override Seed=false
    Sphere Radius=&RandomRangeFloat(Minimum=15, Maximum=25)
    UseOwnerScale=true
   )
  ]
  particleUpdateScript [
   &ParticleState()
   &Color(Color=&MakeCustomLinearColorFromBool(Bool=$User.bIsPainted, True Color=$User.StarBitsColor, False Color=<rgba 0.851 0.5647 0.2235 1>))
   &ScaleSpriteSize(
    Non-Uniform Curve Sprite Scale=DataInterface() {propertyValues {
      XCurve {keys [<at 0 3 RCIM_Cubic>, {tangentMode=RCTM_User, time=1, value=0}]}
      YCurve {keys [{interpMode=RCIM_Cubic, tangentMode=RCTM_User, time=0, value=3}, {tangentMode=RCTM_User, time=1, value=0}]}
     }}
    Scale Sprite Size Mode=::NewEnumerator3
   )
   &CurlNoiseForce(Bias Noise=false, Pan Noise=false)
   &DynamicMaterialParameters(
    Index 0 Param 1=&RandomRangeFloat(Minimum=1, Maximum=6)
    Index 0 Param 2=&RandomRangeFloat(Minimum=0.5, Maximum=1)
    Index 0 Param 3=&RandomRangeFloat(Minimum=0, Maximum=1)
    Index 0 Param 4=2
    Module.Index0Param0Write=true
    Module.Index0Param1Write=true
    Module.Index0Param2Write=true
    Module.Index0Param3Write=true
   )
   &SolveForcesAndVelocity(Clamp Velocity=false, Limit Acceleration=false)
  ]
  rendererClasses [&NiagaraSpriteRendererProperties]
  renderers [{rendererClass=&NiagaraSpriteRendererProperties}]
 }
}
summary {
 systemName=NS_Boost_AstralRadiance
 emitters [
  {emitterName=NE_VC_Boost_Sprite_Parent, rendererClasses [&NiagaraSpriteRendererProperties]}
  {emitterName=NE_VC_Boost_Ribbon_Parent001, rendererClasses [&NiagaraRibbonRendererProperties]}
 ]
 userVariables [
  User.bIsFrontendPreview {type=NiagaraBool}
  User.bIsPainted {type=NiagaraBool}
  User.ColorVariantID {type=NiagaraInt32}
  User.DTCarOffset {type=Vector3f, description="", defaultValue {struct=Vector3f, value=<0 0 0>}}
  User.PaintedColor {type=LinearColor, description="", defaultValue {struct=LinearColor, value=<rgba 1 1 1 1>}}
  User.PaintedScalar {type=NiagaraFloat, defaultValue=1, description=""}
  User.PortCount {type=NiagaraInt32, defaultValue=1}
  User.StarBitsColor {type=LinearColor, description="", defaultValue {struct=LinearColor, value=<rgba 1 1 1 1>}}
  User.ThinTrailColor {type=LinearColor, description="", defaultValue {struct=LinearColor, value=<rgba 1 1 1 1>}}
 ]
}
