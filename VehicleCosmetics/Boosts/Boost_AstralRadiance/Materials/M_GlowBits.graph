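# Translucent unlit Niagara sprite material that draws procedural glow glints for the Astral Radiance vehicle boost cosmetic.
# UV math alone builds two thin crosses masked by a radial falloff: one rotated, one axis-aligned with an added soft core.
# Dynamic parameters shape the glints, blend between both variants and set brightness.
# Partly desaturated particle color tints the emissive through MF_FX_EmissiveAdjustments, and particle alpha scales a separate mask.
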Material /VehicleCosmetics/Boosts/Boost_AstralRadiance/Materials/M_GlowBits
graph {
 properties {
  bAutomaticallySetUsageInEditor=false
  bUsedWithNiagaraSprites=true
  blendMode=BLEND_Translucent
  shadingModel=MSM_Unlit
  usedShadingModels=Unlit
 }
 nodes [
  MaterialExpressionVertexColor_3 = ParticleColor()
  n1 = Add(A=Rotator(Coordinate=TextureCoordinate(), Time=Constant() {r=0.75}) {speed=3.14}, B=Constant() {r=-0.5})
  n2 = Multiply(A=@n1, B=@n1)
  n3 = Constant()
  n4 = Add(A=@n3, B=ComponentMask(None=@n2) {r=true})
  n5 = Constant() {r=60}
  MaterialExpressionConstantClamp_14 = Saturate(None=OneMinus(None=Multiply(A=@n4, B=@n5)))
  n6 = Constant() {r=10}
  n7 = Add(A=ComponentMask(None=@n2) {g=true}, B=@n3)
  MaterialExpressionConstantClamp_15 = Saturate(None=OneMinus(None=Multiply(A=@n7, B=@n5)))
  n8 = Add(A=Panner(Coordinate=TextureCoordinate() {uTiling=3, vTiling=3}, Time=Constant() {r=1}) {speedX=1.5, speedY=1.5}, B=Constant() {r=-3})
  n9 = OneMinus(None=DotProduct(A=@n8, B=@n8))
  MaterialExpressionConstantClamp_18 = Saturate(None=ConstantBiasScale(None=@n9))
  n10 = Multiply(A=@MaterialExpressionConstantClamp_18, B=@MaterialExpressionConstantClamp_18)
  n11 = Multiply(A=Add(A=Power(Base=@MaterialExpressionConstantClamp_14, Exp=@n6), B=Power(Base=@MaterialExpressionConstantClamp_15, Exp=@n6)), B=@n10)
  n12 = DynamicParameter() {paramNames [Power,Sharpness,Switch,Brightness]}
  MaterialExpressionConstantClamp_20 = Saturate(None=Multiply(A=Power(Base=@n11, Exp=@n12.Power), B=@n12.Sharpness))
  MaterialExpressionConstantClamp_19 = Saturate(None=ConstantBiasScale(None=@n9) {bias=-0.3, scale=1.25})
  n13 = Add(A=TextureCoordinate(), B=Constant() {r=-0.5})
  n14 = Multiply(A=@n13, B=@n13)
  n15 = Constant() {r=60}
  n16 = Multiply(A=@n15, B=Add(A=Constant(), B=ComponentMask(None=@n14) {r=true}))
  n17 = Constant() {r=10}
  n18 = Power(Base=Saturate(None=OneMinus(None=@n16)), Exp=@n17)
  n19 = Add(A=ComponentMask(None=@n14) {g=true}, B=Constant())
  MaterialExpressionConstantClamp_16 = Saturate(None=OneMinus(None=Multiply(A=@n19, B=@n15)))
  n20 = Power(Base=@MaterialExpressionConstantClamp_16, Exp=@n17)
  n21 = Power(Base=Multiply(A=Add(A=@n18, B=@n20), B=@n10), Exp=@n12.Power)
  MaterialExpressionConstantClamp_12 = Saturate(None=Add(A=Multiply(A=@n21, B=@n12.Sharpness), B=Multiply(A=@MaterialExpressionConstantClamp_19, B=@MaterialExpressionConstantClamp_19)))
  n22 = LinearInterpolate(A=@MaterialExpressionConstantClamp_20, B=@MaterialExpressionConstantClamp_12, Alpha=@n12.Switch)
  MaterialExpressionConstantClamp_21 = Saturate(None=Multiply(A=@n22, B=@MaterialExpressionVertexColor_3.A))
  MaterialExpressionConstantClamp_22 = Saturate(None=@MaterialExpressionConstantClamp_21)
  n23 = Desaturation(None=@MaterialExpressionVertexColor_3.RGB, Fraction=Constant() {r=0.35})
  MaterialExpressionConstantClamp_13 = Saturate(None=Multiply(A=Multiply(A=@n22, B=@n23), B=@n12.Brightness))
  MaterialFunctionCall(Emissive=@MaterialExpressionConstantClamp_13) {materialFunction=&MF_FX_EmissiveAdjustments}
 ]
}
