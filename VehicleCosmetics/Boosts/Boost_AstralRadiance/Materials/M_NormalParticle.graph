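# Translucent unlit particle material for the Astral Radiance vehicle boost cosmetic, enabled for Niagara sprites, ribbons and mesh particles.
# Particle color RGB feeds the shared MF_FX_EmissiveAdjustments function as emissive.
# A texture-free mask multiplies sine curves of U and V, then raises the product to a high power to form a tight, soft spot centered in UV space.
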Material /VehicleCosmetics/Boosts/Boost_AstralRadiance/Materials/M_NormalParticle
graph {
 properties {
  bAutomaticallySetUsageInEditor=false
  bUsedWithNiagaraMeshParticles=true
  bUsedWithNiagaraRibbons=true
  bUsedWithNiagaraSprites=true
  blendMode=BLEND_Translucent
  shadingModel=MSM_Unlit
  usedShadingModels=Unlit
 }
 nodes [
  MaterialFunctionCall(Emissive=ParticleColor().RGB) {materialFunction=&MF_FX_EmissiveAdjustments}
  n1 = TextureCoordinate()
  n2 = ComponentMask(None=@n1) {r=true}
  n3 = ComponentMask(None=@n1) {g=true}
  Power(Base=Multiply(A=Sine(None=@n2) {period=2}, B=Sine(None=@n3) {period=2})) {constExponent=15}
 ]
}
