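# Translucent unlit material for Niagara ribbon and mesh particle trails in the Astral Radiance vehicle boost cosmetic.
# A panning flow normal map, scaled by the Speed parameter, distorts the tiled UVs of a scrolling trail texture.
# That texture, tinted by particle color, feeds MF_FX_EmissiveAdjustments for glow.
# Opacity math combines the texture with a power-based U mask fading both ends, particle alpha and an Alpha parameter.
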Material /VehicleCosmetics/Boosts/Boost_AstralRadiance/Materials/M_Trail2
graph {
 properties {
  bAutomaticallySetUsageInEditor=false
  bUsedWithNiagaraMeshParticles=true
  bUsedWithNiagaraRibbons=true
  blendMode=BLEND_Translucent
  shadingModel=MSM_Unlit
  usedShadingModels=Unlit
 }
 nodes [
  DynamicParameter() {paramNames [Speed,Distort,Tile,Param4]}
  n1 = TextureCoordinate() {uTiling=1.7, vTiling=1.7}
  n2 = ComponentMask(None=TextureSample(UVs=Panner(Coordinate=@n1) {speedX=-1}) {texture=/DealershipSharedLibrary/FX/Textures/Noise/T_VCVFX_Flow02_N}.RGB) {g=true, r=true}
  n3 = ScalarParameter() {defaultValue=1, parameterName=Speed}
  n4 = Add(A=TextureCoordinate() {uTiling=3}, B=Multiply(A=Multiply(A=@n2, B=@n3), B=Constant() {r=0.8}))
  n5 = AppendVector(A=ScalarParameter() {defaultValue=1, parameterName=Offset}, B=Constant() {r=1})
  n6 = TextureSample(UVs=Panner(Coordinate=Add(A=@n4, B=@n5)) {speedX=-0.9}) {texture=/VehicleCosmetics/Boosts/Boost_AstralRadiance/Textures/Trail4_T}
  MaterialExpressionVertexColor_6 = ParticleColor()
  n7 = Constant() {r=4}
  MaterialFunctionCall(Emissive=Multiply(A=Multiply(A=@n6.RGB, B=@MaterialExpressionVertexColor_6.RGB), B=@n7)) {materialFunction=&MF_FX_EmissiveAdjustments}
  MaterialExpressionVertexColor_7 = ParticleColor()
  n8 = ComponentMask(None=TextureCoordinate()) {r=true}
  n9 = Multiply(A=Power(Base=@n8, Exp=Constant() {r=2}), B=Power(Base=OneMinus(None=@n8), Exp=Constant() {r=2}))
  n10 = Constant() {r=45}
  n11 = Multiply(A=Multiply(A=@n6.RGB, B=Multiply(A=@n9, B=@n10)), B=@MaterialExpressionVertexColor_6.A)
  n12 = ScalarParameter() {defaultValue=1, parameterName=Alpha}
  MaterialExpressionConstantClamp_24 = Saturate(None=Multiply(A=@n7, B=Multiply(A=@n11, B=@n12)))
  Multiply(A=@MaterialExpressionVertexColor_7.A, B=@MaterialExpressionConstantClamp_24)
 ]
}
